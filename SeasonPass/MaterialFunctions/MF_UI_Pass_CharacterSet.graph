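# UI material function for the Season Pass that samples a two-part character texture, with independent transforms for the left and right halves.
# Texture-scaled UVs plus a global offset feed per-half offset, scale and rotation, pivoting on each half's bottom center.
# A rounded U mask selects between the two UV sets for one clamped sample.
# It outputs RGB, the half mask, and alpha edge-clipped through UVs remapped per half.
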MaterialFunction /SeasonPass/MaterialFunctions/MF_UI_Pass_CharacterSet
graph {
 properties {previewMaterialDomain=MD_UI}
 nodes [
  n1 = NamedRerouteUsage()
  FunctionOutput(None=@n1) {outputName=Mask, sortPriority=20}
  n2 = MaterialFunctionCall() {materialFunction=&GetUserInterfaceUV}
  n3 = FunctionInput(Preview=TextureObject()) {inputName=TextureObject, inputType=FunctionInput_Texture2D, sortPriority=-10}
  n4 = Add(
   A=MaterialFunctionCall(
    UV=FunctionInput(Preview=@n2.Normalized UV) {bUsePreviewValueAsDefault=true, inputName=UV, inputType=FunctionInput_Vector2}
    TextureSize=TextureProperty(Tex=@n3)
   ) {materialFunction=&MF_UI_UVScaling}.ScaleY
   B=FunctionInput() {bUsePreviewValueAsDefault=true, inputName=GlobalOffset, inputType=FunctionInput_Vector2, sortPriority=5}
  )
  n5 = FunctionInput() {bUsePreviewValueAsDefault=true, inputName=LeftOffset, inputType=FunctionInput_Vector2, sortPriority=10, previewValue=<0 0 0 0>}
  n6 = Constant2Vector() {g=1, r=0.25}
  n7 = Subtract(A=Add(A=@n4, B=@n5), B=@n6)
  n8 = FunctionInput() {bUsePreviewValueAsDefault=true, inputName=LeftScale, inputType=FunctionInput_Vector2, sortPriority=11, previewValue=<1 1 0 0>}
  n9 = FunctionInput() {bUsePreviewValueAsDefault=true, inputName=LeftRotation, inputType=FunctionInput_Scalar, sortPriority=12, previewValue=<0 0 0 0>}
  n10 = Rotator(Coordinate=Add(A=Multiply(A=@n7, B=@n8), B=@n6), Time=@n9) {centerX=0.25, centerY=1, speed=6.2832}
  n11 = FunctionInput() {bUsePreviewValueAsDefault=true, inputName=RightOffset, inputType=FunctionInput_Vector2, sortPriority=20, previewValue=<0 0 0 0>}
  n12 = Constant2Vector() {g=1, r=0.75}
  n13 = Subtract(A=Add(A=@n4, B=@n11), B=@n12)
  n14 = FunctionInput() {bUsePreviewValueAsDefault=true, inputName=RightScale, inputType=FunctionInput_Vector2, sortPriority=21, previewValue=<1 1 0 0>}
  n15 = FunctionInput() {bUsePreviewValueAsDefault=true, inputName=RightRotation, inputType=FunctionInput_Scalar, sortPriority=22, previewValue=<0 0 0 0>}
  n16 = Rotator(Coordinate=Add(A=Multiply(A=@n13, B=@n14), B=@n12), Time=@n15) {centerX=0.75, centerY=1, speed=6.2832}
  n17 = ComponentMask(None=@n4) {r=true}
  n18 = NamedRerouteDeclaration(None=Round(None=Saturate(None=@n17))) {name=Flip Mask, nodeColor=<rgba 0 0 0 1>}
  n19 = TextureSample(UVs=NamedRerouteDeclaration(None=LinearInterpolate(A=@n10, B=@n16, Alpha=@n18)) {name=Set UV, nodeColor=<rgba 0 0 0 1>}, Tex=@n3) {samplerSource=SSM_Clamp_WorldGroupSettings}
  n20 = NamedRerouteUsage()
  n21 = LinearInterpolate(A=Constant2Vector(), B=Constant2Vector() {r=1}, Alpha=@n1)
  n22 = Add(A=Multiply(A=Subtract(A=@n20, B=@n21), B=Constant2Vector() {g=1, r=2}), B=@n21)
  n23 = NamedRerouteUsage()
  Multiply(
   A=FunctionOutput(None=@n19.RGB) {outputName=RGB}
   B=FunctionOutput(None=MaterialFunctionCall!(UVs=@n22, Alpha=@n23) {materialFunction=/FortUILibrary/MaterialFunctions/Basics/MF_UI_ClipEdgePixel}.Result) {outputName=Alpha, sortPriority=10}
  )
  NamedRerouteDeclaration(None=@n19.A) {name=Alpha, nodeColor=<rgba 0 0 0 1>}
 ]
}
