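# UI material function for the Battle Pass reboot landing page overlay, outputting a triangular shape mask, an intro animation value and a vertex offset.
# Five left and right triangles are built from subtracted, rotated half-plane distance fields with anti-aliasing, on scaled, shifted UVs.
# The reveal value comes from a pass collection intro parameter, with Post Match parameters overriding it and adjusting the mask.
# The vault FocalPointPosition alpha drives a viewport-scaled vertical offset.
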MaterialFunction /SeasonPass/MaterialFunctions/MF_UI_BPReboot_LandingPage_Overlay
graph {
 properties {previewMaterialDomain=MD_UI}
 nodes [
  Constant() {r=0.012}
  Constant() {r=0.8647}
  Constant() {r=1}
  Constant()
  Constant() {r=0.1275}
  Constant() {r=0.2525}
  Constant() {r=0.875}
  n1 = NamedRerouteUsage()
  n2 = LinearInterpolate(A=NamedRerouteUsage(), Alpha=NamedRerouteUsage())
  n3 = NamedRerouteUsage()
  n4 = NamedRerouteUsage()
  FunctionOutput(None=LinearInterpolate(A=LinearInterpolate(A=@n1, Alpha=@n2), Alpha=LinearInterpolate(A=@n3, Alpha=@n4))) {outputName=Shape Mask, sortPriority=1}
  n5 = CollectionParameter() {collection=&/MPC_UI_Pass_Properties, parameterName=Anim_IntroSequence4}
  n6 = ScalarParameter() {group=Post Match, parameterName=PM_Transition_RevealOverride, sliderMax=1}
  FunctionOutput(None=NamedRerouteDeclaration(None=LinearInterpolate(A=@n5, Alpha=@n6)) {name=RevealAmount, nodeColor=<rgba 0 0 0 1>}) {outputName=Anim Intro Value, sortPriority=100}
  n7 = CollectionParameter() {
   collection=/Game/Environments/FrontEnd/Blueprints/ItemPreview/VaultWorldTransitions/MaterialCollectionParameters/MPC_UI_Master_VaultWorld_Settings
   parameterName=FocalPointPosition
  }
  n8 = AppendVector(A=Constant(), B=LinearInterpolate(Alpha=ComponentMask(None=@n7) {a=true}) {constB=144})
  n9 = MaterialFunctionCall() {materialFunction=/FortUILibrary/MaterialFunctions/Basics/MF_UI_ViewportScale}
  FunctionOutput(None=Add(A=WorldPosition().XY, B=Multiply(A=@n8, B=@n9.Viewport Scale Factor))) {outputName=Vertex Offset, sortPriority=20}
  LinearInterpolate(Alpha=SmoothStep(Value=NamedRerouteUsage()) {constMax=0.5}) {constA=-1, constB=0}
  LinearInterpolate(Alpha=SmoothStep(Value=NamedRerouteUsage()) {constMin=0.4}) {constA=-1, constB=0}
  LinearInterpolate(Alpha=SmoothStep(Value=NamedRerouteUsage()) {constMax=0.8, constMin=0.4}) {constA=-1, constB=0}
  LinearInterpolate(Alpha=SmoothStep(Value=NamedRerouteUsage()) {constMax=0.6, constMin=0.1}) {constA=-1, constB=0}
  Multiply(A=ScalarParameter() {defaultValue=1, parameterName=Intro4_P, sliderMax=1}, B=@n5)
  n10 = NamedRerouteUsage()
  n11 = Constant() {r=0.233}
  n12 = Constant() {r=0.25}
  n13 = Constant()
  n14 = ComponentMask(None=Subtract(A=Rotator(Coordinate=@n10, Time=@n11) {speed=6.2832}, B=AppendVector(A=@n12, B=@n13))) {r=true}
  n15 = Constant() {r=0.23}
  n16 = LinearInterpolate(Alpha=ScalarParameter() {group=Post Match, parameterName=PM_Transition_MaskFix, sliderMax=1}) {constA=0.012, constB=-0.6}
  n17 = Constant()
  n18 = ComponentMask(None=Subtract(A=Rotator(Coordinate=@n10, Time=@n15) {speed=6.2832}, B=AppendVector(A=@n16, B=@n17))) {r=true}
  NamedRerouteDeclaration(None=MaterialFunctionCall(DistanceField=MaterialFunctionCall!(DistanceFieldA=@n14, DistanceFieldB=@n18, BlendAmount) {materialFunction=/FortUILibrary/MaterialFunctions/Shapes/Modifiers/MF_UI_SDF_Subtraction}.Exact) {materialFunction=&MF_UI_AntiAliasing}.Inner) {name=Left Triangle 1, nodeColor=<rgba 0 0 0 1>}
  n19 = NamedRerouteUsage()
  n20 = Constant() {r=0.2365}
  n21 = Constant() {r=0.0425}
  n22 = Constant()
  n23 = ComponentMask(None=Subtract(A=Rotator(Coordinate=@n19, Time=@n20) {speed=6.2832}, B=AppendVector(A=@n21, B=@n22))) {r=true}
  n24 = Constant() {r=0.4495}
  n25 = Constant() {r=0.6325}
  n26 = Constant()
  n27 = ComponentMask(None=Subtract(A=Rotator(Coordinate=@n19, Time=@n24) {speed=6.2832}, B=AppendVector(A=@n25, B=@n26))) {r=true}
  NamedRerouteDeclaration(None=MaterialFunctionCall(DistanceField=MaterialFunctionCall(DistanceFieldA=@n23, DistanceFieldB=@n27) {materialFunction=&MF_UI_SDF_Subtraction}.Exact) {materialFunction=&MF_UI_AntiAliasing}.Inner) {name=Left Triangle 2, nodeColor=<rgba 0 0 0 1>}
  n28 = NamedRerouteUsage()
  n29 = Constant() {r=-0.2175}
  n30 = Saturate(None=ScalarParameter() {defaultValue=0.5, group=Post Match, parameterName=PM_Transition_MaskHeight})
  n31 = LinearInterpolate(
   B=LinearInterpolate(Alpha=Multiply(A=@n30) {constB=2}) {constA=0.8647, constB=0.88}
   Alpha=ScalarParameter() {group=Post Match, parameterName=PM_Transition_RevealOverride, sliderMax=1}
  ) {constA=0.8647}
  n32 = Constant()
  n33 = ComponentMask(None=Subtract(A=Rotator(Coordinate=@n28, Time=@n29) {speed=6.2832}, B=AppendVector(A=@n31, B=@n32))) {r=true}
  n34 = Constant() {r=-0.2072}
  n35 = Constant() {r=0.85}
  n36 = Constant()
  n37 = ComponentMask(None=Subtract(A=Rotator(Coordinate=@n28, Time=@n34) {speed=6.2832}, B=AppendVector(A=@n35, B=@n36))) {r=true}
  NamedRerouteDeclaration(None=MaterialFunctionCall(DistanceField=MaterialFunctionCall(DistanceFieldA=@n33, DistanceFieldB=@n37) {materialFunction=&MF_UI_SDF_Subtraction}.Exact) {materialFunction=&MF_UI_AntiAliasing}.Inner) {name=Right Triangle 1, nodeColor=<rgba 0 0 0 1>}
  n38 = NamedRerouteUsage()
  n39 = Constant() {r=-0.2825}
  n40 = Constant() {r=1.052}
  n41 = Constant()
  n42 = ComponentMask(None=Subtract(A=Rotator(Coordinate=@n38, Time=@n39) {speed=6.2832}, B=AppendVector(A=@n40, B=@n41))) {r=true}
  n43 = Constant() {r=-0.2235}
  n44 = Constant() {r=0.85}
  n45 = Constant()
  n46 = ComponentMask(None=Subtract(A=Rotator(Coordinate=@n38, Time=@n43) {speed=6.2832}, B=AppendVector(A=@n44, B=@n45))) {r=true}
  NamedRerouteDeclaration(None=MaterialFunctionCall(DistanceField=MaterialFunctionCall(DistanceFieldA=@n42, DistanceFieldB=@n46) {materialFunction=&MF_UI_SDF_Subtraction}.Exact) {materialFunction=&MF_UI_AntiAliasing}.Inner) {name=Right Triangle 2, nodeColor=<rgba 0 0 0 1>}
  n47 = NamedRerouteUsage()
  n48 = Constant() {r=-0.2125}
  n49 = Constant() {r=0.801}
  n50 = Constant()
  n51 = ComponentMask(None=Subtract(A=Rotator(Coordinate=@n47, Time=@n48) {speed=6.2832}, B=AppendVector(A=@n49, B=@n50))) {r=true}
  n52 = Constant() {r=-0.2}
  n53 = Constant() {r=0.8365}
  n54 = Constant()
  n55 = ComponentMask(None=Subtract(A=Rotator(Coordinate=@n47, Time=@n52) {speed=6.2832}, B=AppendVector(A=@n53, B=@n54))) {r=true}
  n56 = Multiply(A=Saturate(None=ScalarParameter() {defaultValue=0.5, group=Post Match, parameterName=PM_Transition_MaskHeight})) {constB=2}
  n57 = SmoothStep(Value=LinearInterpolate(B=OneMinus(None=@n56), Alpha=ScalarParameter() {group=Post Match, parameterName=PM_Transition_RevealOverride, sliderMax=1}) {constA=1}) {constMin=0.8}
  NamedRerouteDeclaration(None=MaterialFunctionCall(
    DistanceField=MaterialFunctionCall(DistanceFieldA=@n51, DistanceFieldB=@n55) {materialFunction=&MF_UI_SDF_Subtraction}.Exact
    Threshold=LinearInterpolate(Alpha=@n57) {constA=-1, constB=0}
   ) {materialFunction=&MF_UI_AntiAliasing}.Inner) {name=Right Triangle 3, nodeColor=<rgba 0 0 0 1>}
  n58 = MaterialFunctionCall(
   UV=MaterialFunctionCall() {materialFunction=&GetUserInterfaceUV}.Normalized UV
   ScalePivotPoint=Constant2Vector() {g=1, r=0.5}
  ) {materialFunction=&MF_UI_UVScaling}
  n59 = Saturate(None=ScalarParameter() {defaultValue=0.5, group=Post Match, parameterName=PM_Transition_MaskHeight})
  n60 = ScalarParameter() {defaultValue=0.5, group=Post Match, parameterName=PM_Transition_MaskHeight}
  n61 = ScalarParameter() {group=Post Match, parameterName=PM_Transition_RevealOverride, sliderMax=1}
  n62 = NamedRerouteUsage()
  n63 = AppendVector(
   A=LinearInterpolate(B=Multiply(A=@n59) {constB=0.646}, Alpha=ScalarParameter() {group=Post Match, parameterName=PM_Transition_RevealOverride, sliderMax=1})
   B=LinearInterpolate(B=LinearInterpolate(B=@n60, Alpha=@n61), Alpha=@n62) {constA=-0.1, constB=0}
  )
  NamedRerouteDeclaration(None=MaterialFunctionCall(UVs=Add(A=@n58.ScaleY, B=@n63), Texture Scale=LinearInterpolate(Alpha=@n62) {constA=2}) {materialFunction=&ScaleUVsByCenter}.UVs) {name=UV, nodeColor=<rgba 0 0 0 1>}
  NamedRerouteUsage()
  SmoothStep(Value=MaterialFunctionCall(Speed=Constant() {r=0.3}) {materialFunction=&DebugTimeSine}.Result)
 ]
}
