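# Material function for Season Pass UI materials that converts an OkLab color to linear sRGB.
# A single Custom HLSL node maps the Lab input to non-linear LMS, cubes it, then applies the LMS-to-RGB matrix coefficients.
# The result is saturated to the 0–1 range, so out-of-gamut colors clip per channel, and no gamma encoding is applied.
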MaterialFunction /SeasonPass/MaterialFunctions/MF_UI_SeasonPass_OkLab_to_RGB
graph {
 properties {previewBlendMode=BLEND_Translucent, previewMaterialDomain=MD_UI}
 nodes [
  MaterialExpressionFunctionInput_0 = FunctionInput() {previewValue=<0.5 0.8 0.3 1>}
  n1 = Custom!(Lab=@MaterialExpressionFunctionInput_0) {
   code="// OkLab to Linear sRGB\r\nfloat3 LMS_Mix = float3(\r\n  Lab.x + 0.3963377774 * Lab.y + 0.2158037573 * Lab.z,\r\n  Lab.x - 0.1055613458 * Lab.y - 0.0638541728 * Lab.z,\r\n  Lab.x - 0.0894841775 * Lab.y - 1.2914855480 * Lab.z\r\n);\r\n\r\nLMS_Mix = LMS_Mix * LMS_Mix * LMS_Mix;\r\n\r\nfloat3 RGB = float3(\r\n  4.0767416621 * LMS_Mix.r - 3.3077115913 * LMS_Mix.g + 0.2309699292 * LMS_Mix.b,\r\n  -1.2684380046 * LMS_Mix.r + 2.6097574011 * LMS_Mix.g - 0.3413193965 * LMS_Mix.b,\r\n  -0.0041960863 * LMS_Mix.r - 0.7034186147 * LMS_Mix.g + 1.7076147010 * LMS_Mix.b\r\n);\r\n\r\nreturn saturate(RGB);"
   description=Converts OkLab to sRGB
   showCode=true
   inputs [{inputName=Lab, input {expression=&self:MaterialExpressionFunctionInput_0}}]
  }
  FunctionOutput(None=@n1)
 ]
}
