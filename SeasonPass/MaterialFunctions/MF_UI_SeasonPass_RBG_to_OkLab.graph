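# Material function in the Season Pass UI that converts a color from linear sRGB to the OkLab color space.
# A single Custom HLSL node applies a matrix to get LMS, takes a cube root, then applies a second matrix to return Lab.
# LMS is clamped at zero before the root to guard against negatives, and no gamma decoding is done, so the input is treated as already linear.
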MaterialFunction /SeasonPass/MaterialFunctions/MF_UI_SeasonPass_RBG_to_OkLab
graph {
 properties {previewBlendMode=BLEND_Translucent, previewMaterialDomain=MD_UI}
 nodes [
  MaterialExpressionFunctionInput_0 = FunctionInput() {previewValue=<0.5 0.8 0.3 1>}
  n1 = Custom!(Color=@MaterialExpressionFunctionInput_0) {
   code="// Convert ColorA Linear sRGB to OkLab\r\nfloat3 LMS = float3(\r\n  0.4122214708 * Color.r + 0.5363325363 * Color.g + 0.0514459929 * Color.b,\r\n  0.2119034982 * Color.r + 0.6806995451 * Color.g + 0.1073969566 * Color.b,\r\n  0.0883024619 * Color.r + 0.2817188376 * Color.g + 0.6299787005 * Color.b\r\n);\r\n\r\n// Cube root nonlinearly to guard against negative\r\nfloat3 LMS_ = pow(max(LMS, 0.0), 1.0 / 3.0);\r\n\r\n// LMS to OkLab\r\nfloat3 Lab = float3(\r\n  0.2104542553 * LMS_.r + 0.7936177850 * LMS_.g - 0.0040720468 * LMS_.b,\r\n  1.9779984951 * LMS_.r - 2.4285922050 * LMS_.g + 0.4505937099 * LMS_.b,\r\n  0.0259040371 * LMS_.r + 0.7827717662 * LMS_.g - 0.8086757660 * LMS_.b\r\n);\r\n\r\nreturn Lab;"
   description=Converts sRGB to OkLab
   showCode=true
   inputs [{inputName=Color, input {expression=&self:MaterialExpressionFunctionInput_0}}]
  }
  FunctionOutput(None=@n1)
 ]
}
